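# Material function for the AuraCop character that turns the incoming material to gold, with a glowing transition.
# MF_AuraCop_GoldMask, fed pre-skinned position and normal, supplies mask, emissive and time signals.
# A metallic gold layer, tinted by clothing and skin gradients, upward-facing normals and tattoo darkening, blends over the input attributes, gated by GM_CurrentStage.
# Transition emissive is then added and scaled by an inverted mask before output.
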MaterialFunction /BRCosmetics/Characters/Player/Male/Medium/Bodies/M_MED_AuraCop_Bullet/Materials/MaterialFunctions/MF_AuraCop_Reactivity
graph {nodes [
  n1 = NamedRerouteUsage()
  n2 = Saturate(None=Add(A=ComponentMask(None=@n1) {g=true}, B=ScalarParameter() {defaultValue=0.3, group=AuraCop, parameterName=GMapIntensity}))
  n3 = ScalarParameter() {defaultValue=0.68, group=AuraCop, parameterName=TattooGoldDarkness}
  n4 = NamedRerouteUsage()
  n5 = SmoothStep(
   Min=ScalarParameter() {defaultValue=0.5, group=AuraCop, parameterName=GoldBaseColor_Min}
   Max=ScalarParameter() {defaultValue=1, group=AuraCop, parameterName=GoldBaseColor_Max}
   Value=ComponentMask(None=NamedRerouteUsage()) {g=true}
  ) {constMin=0.5}
  n6 = ComponentMask(None=VertexInterpolator(VS=PreSkinnedNormal()).PS) {b=true}
  n7 = LinearInterpolate(
   A=LinearInterpolate(
    A=VectorParameter() {group=AuraCop, parameterName=GoldBaseColor_Clothing_Dark, defaultValue=<rgba 0.5 0.221 0.0183 1>}.RGB
    B=VectorParameter() {group=AuraCop, parameterName=GoldBaseColor_Clothing_Light, defaultValue=<rgba 0.7969 0.4644 0.1038 1>}.RGB
    Alpha=@n5
   )
   B=VectorParameter() {group=AuraCop, parameterName=GoldBaseColor_Y_Top, defaultValue=<rgba 0.8073 0.5146 0.1598 1>}.RGB
   Alpha=SmoothStep(
    Min=ScalarParameter() {defaultValue=0.5, group=AuraCop, parameterName=GoldBaseColor_Min}
    Max=ScalarParameter() {defaultValue=1, group=AuraCop, parameterName=GoldBaseColor_Max}
    Value=Saturate(None=@n6)
   ) {constMin=0.5}
  )
  n8 = LinearInterpolate(
   A=VectorParameter() {group=AuraCop, parameterName=GoldBaseColor_Skin_Dark, defaultValue=<rgba 0.5 0.221 0.0183 1>}.RGB
   B=VectorParameter() {group=AuraCop, parameterName=GoldBaseColor_Skin_Light, defaultValue=<rgba 0.7969 0.4644 0.1038 1>}.RGB
   Alpha=@n5
  )
  n9 = ComponentMask(None=NamedRerouteUsage()) {b=true}
  n10 = Constant() {r=1}
  n11 = Constant()
  n12 = SetMaterialAttributes!(
   MaterialAttributes=NamedRerouteUsage()
   BaseColor=Multiply(A=LinearInterpolate(A=@n2, B=@n3, Alpha=@n4) {constA=1, constB=0.68}, B=LinearInterpolate(A=@n7, B=@n8, Alpha=@n9))
   Specular=@n10
   Roughness=LinearInterpolate(Alpha=NamedRerouteUsage()) {constA=0.25, constB=0.1}
   Metallic=@n10
   EmissiveColor=@n11
   Normal
   Opacity=@n11
   ShadingModel=ShadingModel()
  ) {attributeSetTypes [
    #1
    #2
    #3
    #4
    #5
    #6
    #7
    #8
   ]}
  n13 = NamedRerouteUsage()
  n14 = Constant()
  n15 = If(
   A=ScalarParameter() {defaultValue=100, group=MidasTech_GoldMask, parameterName=GM_CurrentStage, sortPriority=3}
   "A > B"=Constant() {r=1}
   "A == B"=@n14
   "A < B"=@n14
  )
  n16 = GetMaterialAttributes(None=BlendMaterialAttributes(A=NamedRerouteUsage(), B=@n12, Alpha=Multiply(A=@n13, B=@n15))) {attributeGetTypes [#5, #4]}
  n17 = NamedRerouteUsage()
  n18 = VectorParameter() {group=MidasTech_GoldMask, parameterName=GM_Transition_EmissiveColor, sortPriority=101, defaultValue=<rgba 20 8.8102 1 1>}
  n19 = NamedRerouteUsage()
  n20 = Multiply(A=@n18.RGB, B=SmoothStep(Value=OneMinus(None=@n19)) {constMax=0.2, constValue=1})
  n21 = ScalarParameter() {defaultValue=1, group=MidasTech_GoldMask, parameterName=GM_Transition_EmissiveIntensity}
  n22 = LinearInterpolate(A=@n17, B=Add(A=@n17, B=Multiply(A=@n20, B=@n21)), Alpha=NamedRerouteUsage())
  n23 = OneMinus(None=NamedRerouteUsage())
  FunctionOutput(None=SetMaterialAttributes!(MaterialAttributes=@n16.MaterialAttributes, EmissiveColor=Multiply(A=@n22, B=@n23)) {attributeSetTypes [#5]})
  n24 = LocalPosition() {localOrigin=InstancePreSkinning}
  n25 = PreSkinnedNormal()
  n26 = MaterialFunctionCall!(
   Local Position=ShaderStageSwitch(PixelShader=VertexInterpolator(VS=@n24.XYZ).PS, VertexShader=@n24.XYZ)
   Local Normal=ShaderStageSwitch(PixelShader=VertexInterpolator(VS=@n25).PS, VertexShader=@n25)
  ) {materialFunction=&/MF_AuraCop_GoldMask}
  NamedRerouteDeclaration(None=Saturate(None=ComponentMask(None=@n26.GoldMask) {r=true})) {name=PEV_Mask, nodeColor=<rgba 0 0.0422 1 1>}
  NamedRerouteDeclaration(None=@n16.Metallic) {name=CurrentMetallic, nodeColor=<rgba 0 1 0.6953 1>}
  n27 = FunctionInput() {bUsePreviewValueAsDefault=true, inputName=InTattooMask, inputType=FunctionInput_Scalar, sortPriority=10, previewValue=<0 0 0 0>}
  NamedRerouteDeclaration(None=@n27) {name=InTattooMask, nodeColor=<rgba 0 1 0.5078 1>}
  NamedRerouteDeclaration(None=@n26.TimeNative) {name=TimeNative, nodeColor=<rgba 0.0547 1 0 1>}
  NamedRerouteDeclaration(None=Saturate(None=ComponentMask(None=@n26.GoldMask) {g=true})) {name=PEV_Emissive, nodeColor=<rgba 0 0.1562 1 1>}
  NamedRerouteDeclaration(None=@n16.EmissiveColor) {name=CurrentEmissive, nodeColor=<rgba 0.9922 1 0 1>}
  NamedRerouteDeclaration(None=FunctionInput() {inputName=MaterialAttributes, inputType=FunctionInput_MaterialAttributes, sortPriority=10}) {name=InMaterialAttributes, nodeColor=<rgba 1 0.7266 0 1>}
  NamedRerouteDeclaration(None=FunctionInput() {bUsePreviewValueAsDefault=true, inputName=Mask, sortPriority=10, previewValue=<1 1 1 1>}) {name=inMask, nodeColor=<rgba 1 0.1875 0 1>}
 ]}
